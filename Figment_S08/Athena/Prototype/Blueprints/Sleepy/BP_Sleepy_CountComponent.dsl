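; Blueprint component for the prototype Sleepy feature that relays damage notifications to listeners.
; The ServerDamageTaken custom event receives an integer damage value and passes it straight to the OnDamageTaken event dispatcher.
; BeginPlay and Tick are empty, and despite the component's name, the graph shows no counting logic.
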(event Custom|ServerDamageTaken
  (params (Damage :type Integer))
  (Default|CallOnDamageTaken Damage))

(event EventBeginPlay)

(event EventTick
  (params (DeltaSeconds :type Float%20%28single-precision%29)))
