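; BP_Sleepy_Prop is a prototype big-health building prop that never dies and travels along a spline as server-set progress changes.
; It reports damage through a dispatcher, while leaf-server events set replicated health and progress.
; A replicated spline position drives a timeline move on each machine, during which authority capsule overlaps launch players not riding the pedestal and destroy damageable buildings.
; Clients add movement audio, material fades and hit-wave effects.
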(event Custom|ClientMoveStart
  ;[1 1 1] Client Start and End Move Events
  (Utilities|FlowControl|Sequence
    (:then_0
      (Components|Activation|Activate (Variables|Default|GetP_Sleepy_Move_Start) false)
      (CallFunction|SetMovementAudioEnabled true)
      (Audio|Components|Audio|FadeOut (Variables|Default|GetAmbientAudio) 1.0 0.0 "Linear" :!class /Script/Engine.AudioComponent))
    (:then_1
      (bind (_direction _opacity) (Timeline|Timeline_0 PlayFromStart)
        (:Update
          (Rendering|Material|SetScalarParameterValueonMaterials (Variables|Default|Getsm_sleepy_tech_lines) "Final_Opacity" _opacity)
          (Rendering|Material|SetScalarParameterValueonMaterials (Variables|Default|GetPedestal_FieldFX) "Opacity" (Math|Float|Lerp 1.0 0.0 _opacity)))))))

(event Custom|ClientMoveEnd
  ;[1 1 1] Client Start and End Move Events
  (Utilities|FlowControl|Sequence
    (:then_0
      (Timeline|Timeline_0 ReverseFromEnd))
    (:then_1
      (CallFunction|SetMovementAudioEnabled false)
      (Utilities|FlowControl|Delay 5.0)
      (Audio|Components|Audio|FadeIn (Variables|Default|GetAmbientAudio) 3.0 1.0 0.0 "Linear" :!class /Script/Engine.AudioComponent))))

(event Custom|EdgeServerApplyDamage
  (params (TotalHealth :type Integer))
  (bind _returnvalue (Math|Conversions|ToInteger64(Integer) TotalHealth))
  ;[1 1 1] Leaf server only
  (Development|PrintString (Utilities|String|ToString(Integer) TotalHealth) true true "(R=0.000000,G=0.660000,B=1.000000,A=1.000000)" 22.0)
  (Utilities|FlowControl|Sequence
    (:then_1
      (CallFunction|SetBigHealth _returnvalue true)
      (Health|SetHealth (Math|Float|Min(Float) (Math|Conversions|ToFloat(Integer64) _returnvalue) (Health|GetMaxHealth)))
      (Variables|Default|SetLocalPropHealth _returnvalue)
      (Networking|ForceNetUpdate))))

(event Custom|EdgeServerSetProgress
  (params (Progress :type Float%20%28double-precision%29))
  ;[1 1 1] Leaf server only
  (Development|PrintString (Utilities|String|ToString(Float) Progress))
  (if (!= (Variables|Default|GetProgress) Progress)
    (bind _output_get (Variables|Default|SetProgress Progress))
    (CallFunction|MoveAlongSpline _output_get)))

(event EventOnDamageServer
  (params (Damage :type Float%20%28single-precision%29) (DamageTags :type Gameplay%20Tag%20Container%20Structure%20%28by%20ref%29) (Momentum :type Vector :default "0, 0, 0") (HitInfo :type Hit%20Result%20Structure%20%28by%20ref%29) (InstigatedBy :type Controller%20Object%20Reference) (DamageCauser :type Actor%20Object%20Reference) (EffectContext :type Gameplay%20Effect%20Context%20Handle%20Structure))
  (|CallParentFunction Damage DamageTags Momentum HitInfo InstigatedBy DamageCauser EffectContext)
  (if (not (Utilities|Equal(Object) DamageCauser self))
    (Default|CallOnSleepyPropDamaged (Math|Float|Truncate Damage))))

(event Custom|MoveAlongSpline
  (params (Position :type Float%20%28double-precision%29))
  (Variables|Default|SetCurrentSplinePosition (Variables|Default|GetNewSplinePosition))
  (Variables|Default|SetNewSplinePosition Position)
  (Networking|FlushNetDormancy)
  (CallFunction|MoveThing))

(event Custom|MoveThing
  (bind _uniquetargets (Variables|Default|GetUniqueTargets))
  (bind _currentsplineposition (Variables|Default|GetCurrentSplinePosition))
  (bind _newsplineposition (Variables|Default|GetNewSplinePosition))
  (bind _spline (Variables|Components|GetSpline))
  (Fortnite|SetCanbeDamaged self)
  (Utilities|Array|Clear _uniquetargets)
  (Utilities|Array|Add _uniquetargets self)
  (bind (_direction _newtrack_0) (Timeline|Move PlayFromStart)
    (:Update
      (Utilities|FlowControl|Sequence
        (:then_0
          _)
        (:then_1
          (Transformation|SetActorTransform 0 (Math|Transform|MakeTransform (Spline|GetLocationatTime _spline (Math|Float|Lerp _currentsplineposition _newsplineposition _newtrack_0) "World" true) (Math|Rotator|MakeRotfromZX (0.0 0.0 1.0) (* (Spline|GetTangentatTime _spline (Math|Float|Lerp _currentsplineposition _newsplineposition _newtrack_0) "World" true) -1.0))) false false)
          (Rendering|Material|SetScalarParameterValueonMaterials (Variables|Default|Getsm_sleepy_tech_lines) "Active" 1.0) ; Techlines turn blue
          (Audio|Parameter|SetFloatParameter (Variables|Default|GetMovementAudio) "Progress" _newtrack_0 :!class /Script/Engine.AudioComponent)
          (Networking|SwitchHasAuthority
            (:Authority
              (bind _outactors (Collision|CapsuleOverlapActors (Spline|GetLocationatTime _spline (Math|Float|Lerp _currentsplineposition _newsplineposition _newtrack_0) "World" true) 175.0 6.5e+02 (Utilities|Array|MakeArray "ObjectTypeQuery1" "ObjectTypeQuery3") 0 (Utilities|Array|MakeArray self)))
              (for _array_element _outactors
                (bind _returnvalue (Utilities|Array|AddUnique (Variables|Default|GetUniqueTargets) _array_element))
                (if (Math|Conversions|ToBoolean(Integer) _returnvalue)
                  (Development|PrintString (Utilities|GetDisplayName _array_element))
                  (bind _asfort_player_pawn (Utilities|Casting|CastToFortPlayerPawn _array_element)
                    (:then
                      (if (not (Utilities|Equal(Object) (Pawn|Components|CharacterMovement|GetMovementBaseObject (Variables|/Script/Engine.Character|GetCharacterMovement _asfort_player_pawn)) (Variables|Components|GetLakeStructure_Pedestal)))
                        (Pawn|Character|LaunchCharacterJump _asfort_player_pawn (+ (* (Math|Vector|GetForwardVector (Math|Rotator|FindLookatRotation (Spline|GetLocationatTime _spline (Math|Float|Lerp _currentsplineposition _newsplineposition _newtrack_0) "World" true) (Transformation|GetActorLocation _asfort_player_pawn))) 1.5e+03) "0, 0,500.000000") false false true true)))
                    (:CastFailed
                      (bind _asbuilding_actor (Utilities|Casting|CastToBuildingActor _array_element))
                      (if (Variables|/Script/Engine.Actor|GetbCanBeDamaged _asbuilding_actor)
                        (Actor|DestroyActor _array_element)))))))))))
    (:Finished
      ;[1 1 1] Call start and end
      (Utilities|FlowControl|Sequence
        (:then_0
          (Networking|CanExecuteCosmeticEvents
            (:True
              (CallFunction|ClientMoveEnd))
            (:False
              (Fortnite|SetCanbeDamaged self true))))
        (:then_1
          (Utilities|FlowControl|Delay 0.0)))))
  (Utilities|FlowControl|DoOnce false
    (:Completed
      (Networking|CanExecuteCosmeticEvents
        (:True
          (CallFunction|ClientMoveStart))))))

(event EventBeginPlay
  (bind _pedestal_runefx (Variables|Default|GetPedestal_RuneFX))
  (Utilities|FlowControl|Sequence
    (:then_0
      (Collision|SetCollisionResponsetoChannel (Variables|Debug|GetStaticMeshComponent) "ECC_GameTraceChannel13" "ECR_Block")
      (Class|FortAsyncActionCurrentPlaylistReady|PlaylistDataReadyAsync
        (:OnReady
          (if (== (ScalableFloat|GetValueatLevel (Variables|Default|GetHotfixNoDamage)) 1.0)
            (Fortnite|SetCanbeDamaged self)
            (Variables|BuildingPropBigHealth|SetbShowHealthBar false)
            (else
              (Networking|SwitchHasAuthority
                (:Authority
                  (GameplayEffects|ApplyGameplayEffectToSelf :self (Ability|GetAbilitySystemComponent self) :GameplayEffectClass "/Figment_S08/Athena/Prototype/Blueprints/Iceberg/GE_MaxHealthSet.GE_MaxHealthSet_C" :Level 0.0))))))))
    (:then_1
      (Rendering|Material|SetMaterial _pedestal_runefx 0 "/Figment_S08/Athena/Effects/Materials/Snow/MI_Snow_Rune_01.MI_Snow_Rune_01")
      (Rendering|Material|SetMaterial _pedestal_runefx 1 "/Figment_S08/Athena/Effects/Materials/Snow/MI_Snow_Rune_01.MI_Snow_Rune_01"))))

(event Custom|SetSleepyPropEnabled
  (params (Enabled :type Boolean))
  (bind _ambientaudio (Variables|Default|GetAmbientAudio))
  (if Enabled
    (Collision|SetActorEnableCollision true)
    (Rendering|SetActorHiddenInGame false)
    (Audio|Components|Audio|FadeIn _ambientaudio 1.0 1.0 0.0 "Linear" :!class /Script/Engine.AudioComponent)
    (else
      (Collision|SetActorEnableCollision false)
      (Rendering|SetActorHiddenInGame true)
      (Audio|Components|Audio|FadeOut _ambientaudio 1.0 0.0 "Linear" :!class /Script/Engine.AudioComponent))))

(event Collision|EventHit
  (params (MyComp :type Primitive%20Component%20Object%20Reference) (Other :type Actor%20Object%20Reference) (OtherComp :type Primitive%20Component%20Object%20Reference) (bSelfMoved :type Boolean) (HitLocation :type Vector :default "0, 0, 0") (HitNormal :type Vector :default "0, 0, 0") (NormalImpulse :type Vector :default "0, 0, 0") (Hit :type Hit%20Result%20Structure%20%28by%20ref%29))
  (Development|PrintString (Utilities|String|Append (Utilities|GetDisplayName Other) " HIT ME")))

(event Collision|EventActorBeginOverlap
  (params (OtherActor :type Actor%20Object%20Reference)))

(event EventOnDamagePlayEffects
  (params (Damage :type Float%20%28single-precision%29) (DamageTags :type Gameplay%20Tag%20Container%20Structure%20%28by%20ref%29) (Momentum :type Vector :default "0, 0, 0") (HitInfo_bBlockingHit :type Boolean) (HitInfo_bInitialOverlap :type Boolean) (HitInfo_Time :type Float%20%28single-precision%29) (HitInfo_Distance :type Float%20%28single-precision%29) (HitInfo_Location :type Vector :default "0, 0, 0") (HitInfo_ImpactPoint :type Vector :default "0, 0, 0") (HitInfo_Normal :type Vector :default "0, 0, 0") (HitInfo_ImpactNormal :type Vector :default "0, 0, 0") (HitInfo_PhysMat :type Physical%20Material%20Object%20Reference) (HitInfo_HitActor :type Actor%20Object%20Reference) (HitInfo_HitComponent :type Primitive%20Component%20Object%20Reference) (HitInfo_HitBoneName :type Name :default "None") (HitInfo_BoneName :type Name :default "None") (HitInfo_HitItem :type Integer) (HitInfo_ElementIndex :type Integer) (HitInfo_FaceIndex :type Integer) (HitInfo_TraceStart :type Vector :default "0, 0, 0") (HitInfo_TraceEnd :type Vector :default "0, 0, 0") (InstigatedBy :type Fort%20Pawn%20Object%20Reference) (DamageCauser :type Actor%20Object%20Reference) (EffectContext :type Gameplay%20Effect%20Context%20Handle%20Structure))
  (bind _sm_sleepy_tech_lines (Variables|Default|Getsm_sleepy_tech_lines))
  (bind _sm_sleepy_tech_lines_2 (Variables|Default|Getsm_sleepy_tech_lines))
  ;[1 1 1] Hit FX
  (if (not (Components|Timeline|IsPlaying (Variables|Components|GetMove)))
    (bind _output_get (Variables|Default|SetHitImpactLocation HitInfo_ImpactPoint))
    (GameplayCue|ExecuteGameplayCue%28Burst%29Local self (Variables|Default|GetCueTag_SleepyImpact) (Ability|GameplayCue|MakeGameplayCueParameters :NormalizedMagnitude 6e+02 :RawMagnitude Damage :EffectContext "()" :AggregatedSourceTags "(GameplayTags=)" :AggregatedTargetTags "(GameplayTags=)" :Location _output_get :Normal HitInfo_ImpactNormal :Instigator InstigatedBy :EffectCauser self :GameplayEffectLevel 1 :AbilityLevel 1 :bReplicateLocationWhenUsingMinimalRepProxy false))
    (Rendering|Material|SetVectorParameterValueonMaterials _sm_sleepy_tech_lines "HitWave-Loc" (Variables|Default|GetHitImpactLocation))
    (Rendering|Material|SetScalarParameterValueonMaterials _sm_sleepy_tech_lines "GameTime" (Utilities|Time|GetGameTimeinSeconds))
    (Rendering|Material|SetScalarParameterValueonMaterials _sm_sleepy_tech_lines_2 "Active" 0.0)
    (Rendering|Material|SetScalarParameterValueonMaterials _sm_sleepy_tech_lines_2 "Health" (Health|GetHealthPercent))))

(fn OnRep_LocalPropHealth
  (Networking|SwitchHasAuthority
    (:Remote
      (CallFunction|SetBigHealth (Variables|Default|GetLocalPropHealth) true)
      (Building|RequestBuildingFullUpdate))))

(fn OnRep_NewSplinePosition
  (CallFunction|MoveThing))

(fn SetMovementAudioEnabled
  (params (Enabled :type Boolean))
  (bind _lakestructure_pedestal (Variables|Components|GetLakeStructure_Pedestal))
  (bind _returnvalue (Transformation|GetWorldLocation _lakestructure_pedestal))
  (if Enabled
    (Audio|Components|Audio|FadeIn (Variables|Default|GetMovementAudio) 0.25 1.0 0.0 "Linear" :!class /Script/Engine.AudioComponent)
    (Audio|PlaySoundatLocation (Variables|Default|GetSoundOnMoveStart) _returnvalue)
    (else
      (Audio|Components|Audio|FadeOut (Variables|Default|GetMovementAudio) 0.25 0.0 "Linear" :!class /Script/Engine.AudioComponent)
      (Audio|PlaySoundatLocation (Variables|Default|GetSoundOnMoveEnd) _returnvalue))))

(fn ShouldDie
  (params (Damage :type Float%20%28single-precision%29) (EventInstigator :type Controller%20Object%20Reference) (DamageCauser :type Actor%20Object%20Reference) (bIsTest :type Boolean) (EffectContext :type Gameplay%20Effect%20Context%20Handle%20Structure%20%28by%20ref%29))
  (return false))

(fn ConstructionScript
  (bind _pedestal_runefx (Variables|Default|GetPedestal_RuneFX))
  (Utilities|FlowControl|Sequence
    (:then_0
      (Collision|SetActorEnableCollision true)
      (Transformation|SetActorTransform 0 (Math|Transform|MakeTransform (Spline|GetLocationatTime (Variables|Components|GetSpline) (Variables|Default|GetProgress) "World" true) (Math|Rotator|MakeRotfromZX (0.0 0.0 1.0) (* (Spline|GetTangentatTime (Variables|Components|GetSpline) (Variables|Default|GetProgress) "World" true) -1.0))) false false))
    (:then_1
      (Rendering|Material|SetMaterial _pedestal_runefx 0 "/Figment_S08/Athena/Effects/Materials/Snow/MI_Snow_Rune_01.MI_Snow_Rune_01")
      (Rendering|Material|SetMaterial _pedestal_runefx 1 "/Figment_S08/Athena/Effects/Materials/Snow/MI_Snow_Rune_01.MI_Snow_Rune_01"))))
